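; ToyFox event location actor that supplies player, enemy and pedestal placements for a boss encounter.
; When enabled, it registers with the ToyFox game state component on dedicated servers after playlist data and references are ready.
; SpawnEnemy creates a boss spawner provider once at the enemy marker, then broadcasts the spawned boss.
; Editor-only helpers position enemy and pedestal relative to the player marker, tracing for ground; a map icon aids debugging.
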(event EventBeginPlay
  ;[0 0.15 0.06897 0.5] BeginPlay: Register the event location with the game state component.
  (|CallParentFunction)
  (if (Networking|IsDedicatedServer)
    (Class|FortAsyncActionCurrentPlaylistReady|PlaylistDataReadyAsync
      (:OnReady
        (if (ScalableFloat|GetValueasBool (Variables|Default|GetEventLocationEnabled))
          (Default|WaitUntilReferencesReady 20
            (:OutSuccess
              (bind _returnvalue (Fortnite|Athena|GameState|GetGameStateAthena))
              (Utilities|IsValid _returnvalue
                (:Is%20Valid
                  (bind _returnvalue_2 (Actor|GetComponentbyClass _returnvalue "/ToyFoxGameplay/Gameplay/B_ToyFoxGameStateComponent.B_ToyFoxGameStateComponent_C"))
                  (Utilities|IsValid _returnvalue_2
                    (:Is%20Valid
                      (Event|RegisterEventLocationActor _returnvalue_2 self))))))))))))

(event Custom|SpawnEnemy
  ;[0.1096 0.005555 0.15 0.5] SpawnEnemy: Spawns the enemy boss at the desired location.
  (if (Networking|IsDedicatedServer)
    (bind (_bvalid _location _rotation) (CallFunction|GetEnemyLocationandRotation))
    (Utilities|FlowControl|DoOnce false
      (:Completed
        (bind _returnvalue (Game|SpawnActorfromClass "/ToyFoxGameplay/Gameplay/Enemy/BP_StaticProvider_ToyFoxBoss_Spawner.BP_StaticProvider_ToyFoxBoss_Spawner_C" (Math|Transform|MakeTransform _location _rotation) "AlwaysSpawn" "MultiplyWithRoot" 0 0))
        (EventDispatchers|BindEventtoOnActorSpawned _returnvalue (EventDispatchers|CreateEvent :!event Custom|HandleEnemySpawned))
        (Class|BPStaticProviderToyFoxBossSpawner|EnableProvider _returnvalue)))))

(event Custom|HandleEnemySpawned
  (params (EnemyActor :type Actor%20Object%20Reference))
  ;[0.01193 0.15 0.1403 0.5] HandleEnemySpawned: Broadcast that the enemy has spawned.
  (if (Networking|IsDedicatedServer)
    (Utilities|IsValid EnemyActor
      (:Is%20Valid
        (Default|CallOnEnemySpawned EnemyActor)))))

(event Custom|SetMapIconActive
  (params (bActive :type Boolean))
  (bind (_tintcolor _drawas _tiling _mirroring _imagesize _margin _resourceobject _outlinesettings) (Utilities|Struct|BreakSlateBrush (Variables|/Script/FortniteGame.AthenaSpecialActorComponent|GetSpecialActorMinimapIconBrush (Variables|Default|GetAthenaSpecialActor))))
  (bind _athenaspecialactor (Variables|Default|GetAthenaSpecialActor))
  ;[0.1305 0.01923 0.15 0.5] SetMapIconActive: Only used for debugging to find the locations on the map.
  (Networking|SwitchHasAuthority
    (:Authority
      (if bActive
        (Components|Activation|SetActive _athenaspecialactor true false)
        (bind _object (Utilities|AsyncLoadAsset (Variables|Setup|GetSpecialActorMapIcon))
          (:Completed
            (bind _astexture_2d (Utilities|Casting|CastToTexture2D _object))
            (Class|AthenaSpecialActorComponent|SetMiniMapIcon _athenaspecialactor (Utilities|Struct|MakeSlateBrush _tintcolor _drawas _tiling _mirroring _imagesize _margin _astexture_2d _outlinesettings))))
        (else
          (Components|Activation|Deactivate (Variables|Default|GetAthenaSpecialActor)))))))

(fn GetAthenaSpecialActorComponent
  (return (Variables|Default|GetAthenaSpecialActor)))

(fn GetEnemyLocationAndRotation
  (bind _enemyref (Variables|Default|GetEnemyRef)
    (:then
      (bind (_location _rotation) (Math|Transform|BreakTransform (Transformation|GetActorTransform _enemyref)))
      (return true _location _rotation))
    (:else
      (return false "0, 0, 0" "0, 0, 0"))))

(fn GetEventLocationName
  (return (Variables|Default|GetEventLocationName)))

(fn GetPedestalLocation
  (Utilities|FlowControl|Sequence
    (:then_0
      (bind _pedestalref (Variables|Default|GetPedestalRef)
        (:then
          (return true (Transformation|GetActorLocation _pedestalref)))
        (:else)))
    (:then_1
      (return false "0, 0, 0"))))

(fn GetPlayerLocationAndRotation
  (bind _playerref (Variables|Default|GetPlayerRef)
    (:then
      (bind (_location _rotation) (Math|Transform|BreakTransform (Transformation|GetActorTransform _playerref)))
      (return true _location _rotation))
    (:else
      (return false "0, 0, 0" "0, 0, 0"))))

(fn IsEnabled
  (return (ScalableFloat|GetValueasBool (Variables|Default|GetEventLocationEnabled))))

(fn SetupDefaultEnemyLocationAndRotation
  (if (Editor|InEditor)
    (bind _playerref (Variables|Default|GetPlayerRef)
      (:then
        (bind _enemyref (Variables|Default|GetEnemyRef)
          (:then
            (bind (_returnvalue_x _returnvalue_y) (Math|Vector|BreakVector (+ (Transformation|GetActorLocation _playerref) (* (Transformation|GetActorForwardVector _playerref) (Variables|Setup|GetSetupEnemyDistance)))))
            (Transformation|SetActorLocation _enemyref (Math|Vector|MakeVector _returnvalue_x _returnvalue_y (.z (Transformation|GetActorLocation _enemyref))) false false)
            (bind (_returnvalue_roll _returnvalue_pitch _returnvalue_yaw) (Math|Rotator|BreakRotator (Math|Rotator|FindLookatRotation (Transformation|GetActorLocation _enemyref) (Transformation|GetActorLocation _playerref))))
            (Transformation|SetActorRotation _enemyref (Math|Rotator|MakeRotator 0.0 0.0 _returnvalue_yaw) false)
            (Transactions|TransactObject _enemyref))
          (:else)))
      (:else))))

(fn SetupDefaultPedestalLocation
  (if (Editor|InEditor)
    (bind _pedestalref (Variables|Default|GetPedestalRef)
      (:then
        (bind _playerref (Variables|Default|GetPlayerRef)
          (:then
            (bind (_returnvalue_x _returnvalue_y) (Math|Vector|BreakVector (Transformation|GetWorldLocation (Variables|/ToyFoxGameplay/Gameplay/Locations/B_ToyFox_SpawnLocation_Player.B_ToyFox_SpawnLocation_Player_C|GetEDITORONLY_WeaponSkeletalMeshComponentBudgeted _playerref))))
            (bind (_returnvalue_x_2 _returnvalue_y_2 _returnvalue_z) (Math|Vector|BreakVector (Utilities|Operators|Add (* (Transformation|GetActorForwardVector _playerref) (Variables|Setup|GetSetupPedestalDistance)) (Math|Vector|MakeVector _returnvalue_x _returnvalue_y (.z (Transformation|GetActorLocation _playerref))))))
            (bind _outhits (Collision|MultiLineTraceByChannel (Math|Vector|MakeVector _returnvalue_x_2 _returnvalue_y_2 (+ _returnvalue_z 2e+03)) (Math|Vector|MakeVector _returnvalue_x_2 _returnvalue_y_2 (- _returnvalue_z 5e+03)) "TraceTypeQuery10" true (Utilities|Array|MakeArray _playerref) "None" true "(R=1.000000,G=0.000000,B=0.000000,A=1.000000)" "(R=0.000000,G=1.000000,B=0.000000,A=1.000000)" 5.0))
            (for _array_element _outhits
              (if (Collision|BreakHitResult _array_element)
                (if (Math|Boolean|ORBoolean (.bSuccess (Utilities|Casting|CastToFortStaticMeshActor (.HitActor (Collision|BreakHitResult _array_element)))) (.bSuccess (Utilities|Casting|CastToLandscapeProxy (.HitActor (Collision|BreakHitResult _array_element)))) (.bSuccess (Utilities|Casting|CastToLandscape (.HitActor (Collision|BreakHitResult _array_element)))) (.bSuccess (Utilities|Casting|CastToLandscapeSplineActor (.HitActor (Collision|BreakHitResult _array_element)))))
                  (Transformation|SetActorLocation _pedestalref (.ImpactPoint (Collision|BreakHitResult _array_element)) false false)
                  (Transactions|TransactObject _pedestalref)))))
          (:else)))
      (:else))))

(fn ConstructionScript)
